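# Niagara system for the Pumpkin Spice backpack cosmetic that draws faint, steam-like ribbon wisps from the backpack's skeletal mesh.
# The active emitter spawns four short-lived ribbons at skeletal positions, drifting upward with briefly inherited velocity and panning, fading curl noise.
# Alpha fades with age, particle speed and distance from the owner, and an owner-speed threshold kills particles.
# A disabled Steam001 emitter holds an AnimTrail-based version of the effect.
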NiagaraSystem /BRCosmetics/Accessories/FORT_Backpacks/Backpack_F_MED_Pumpkin_Spice/Fx/NS_Backpack_PumpkinSpice
dependencies {
 usedDataInterfaces [&NiagaraDataInterfaceCurve]
 usedDynamicInputs [&UniformRangedFloat]
 usedModules [
  &self:SystemSpawnScript.SystemScriptSource.SystemScriptGraph.NiagaraNodeAssignment_0.SetVariables_#1
  &SystemState
 ]
}
emitters {
 FN_AttachedRibbon_v3 {
  emitterUpdateScript [
   &EmitterState()
   &FN_SpawnRibbon_v2(Number of Ribbons=2, Spawn Mode=::NewEnumerator0, SpawnRate=30) {enabled=false}
   &FN_Integrator() {enabled=false}
   /Game/Effects/Niagara/Modules/Ribbons/FN_SpawnRibbon_v3(
    Number of Ribbons=4
    Skeleton Mesh:NiagaraDataInterfaceSkeletalMesh=$User.SkeletalMesh
    Spawn Mode:ENiagaraSkeletalSpawning=::NewEnumerator1
    SpawnRate=30
   )
  ]
  particleSpawnScript [
   /Game/Effects/Niagara/Modules/Ribbons/FN_InitializeRibbon_v4(
    Color:LinearColor=<rgba 1 1 1 0.15>
    Enable Skeletal Position=true
    Lifetime=0.75
    Offset Amount=-1
    Offset Axis:ENiagaraOrientationAxis=::ZAxis
    Ribbon Width=3
    SkeletalMesh:NiagaraDataInterfaceSkeletalMesh=$User.SkeletalMesh
   )
   &FN_FlipRibbonV() {enabled=false}
   &AddVelocity(Velocity=<0 0 100>) {enabled=false}
   &AddVelocity(Velocity=<0 0 27>) {module=AddVelocity001}
   &Material(PostMult=0.2, PreMult=0)
  ]
  particleUpdateScript [
   /SetVariables(Particles.EngineOwnerSpeed:NiagaraFloat=&VectorLength(Vector=$Engine.Owner.Velocity)) {isSetParameters=true}
   &InheritVelocity(
    Module.Limit Speed=true
    Velocity Limit=500
    Velocity Scale=&MakeVector(
     X=&FloatFromCurve(FloatCurve=DataInterface() {propertyValues {Curve {keys [<at 0 0.75>,<at 0.01 0>]}}}, Scale Curve=0.5)
     Y=&FloatFromCurve(FloatCurve=DataInterface() {propertyValues {Curve {keys [<at 0 0.75>,<at 0.01 0>]}}}, Scale Curve=0.5)
     Z=&FloatFromCurve(FloatCurve=DataInterface() {propertyValues {Curve {keys [<at 0 0.75>,<at 0.01 0>]}}}, Scale Curve=0.1)
    )
   )
   /SetVariables(Particles.Speed:NiagaraFloat=&NormalizeFloat(Value=&VectorLength(Vector=$Particles.Velocity), Min=0, Max=200)) {isSetParameters=true}
   /Niagara/Modules/Update/Forces/CurlNoiseForce(
    Noise Frequency=0.25
    Noise Strength=&FloatFromCurve(FloatCurve=DataInterface() {propertyValues {Curve {keys [<at 0 1>,<at 1 0>]}}}, Scale Curve=150)
    Pan Noise=true
    Pan Noise Field=<0 0 -0.5>
   )
   &SolveForcesAndVelocity(Clamp Velocity=false, Limit Acceleration=false)
   /Game/Effects/Niagara/Modules/Ribbons/FN_UpdateRibbon_v4(Enable Skeletal Position=true, SkeletalMesh:NiagaraDataInterfaceSkeletalMesh=$User.SkeletalMesh)
   /Game/Effects/Niagara/Modules/Materials/FN_RibbonMaterial/FN_RibbonMaterial_v2() {enabled=false}
   &ScaleColor(
    Scale Alpha=&Multiply_Float(
     A=&FloatFromCurve(FloatCurve=DataInterface() {propertyValues {Curve {keys [{arriveTangent=0.002, interpMode=RCIM_Cubic, time=0.0401, value=0.4707}, <at 1.0011 0.002>]}}})
     B=&FloatFromCurve(
      FloatCurve=DataInterface() {propertyValues {Curve {keys [
          {arriveTangent=-2.1907, interpMode=RCIM_Cubic, leaveTangent=-0.886, tangentMode=RCTM_User, time=0, value=1}
          {arriveTangent=-1.1071, interpMode=RCIM_Cubic, leaveTangent=-1.1071, tangentMode=RCTM_User, time=1, value=0}
         ]}}}
      CurveIndex=$Particles.Speed
     )
    )
    ScaleA=true
    ScaleRGB=true
   )
   &ScaleColor(
    Scale Alpha=&Lerp_Float(
     A=1
     B=0
     Alpha=&NormalizeDistanceRange(
      Start Position:NiagaraPosition=$Engine.Owner.Position
      End Position:NiagaraPosition=$Particles.Position
      Distance:NiagaraFloat=50
      Invert Normalized Range:NiagaraBool=false
     )
    )
    ScaleA=true
    ScaleRGB=true
   ) {module=ScaleColor001}
   &KillParticles(Kill Particles=&SetBoolByFloatComparison(A=500, B=$Particles.EngineOwnerSpeed, Comparison Type=::NewEnumerator6))
  ]
  rendererClasses [&NiagaraRibbonRendererProperties]
  renderers [
   {rendererClass=&NiagaraRibbonRendererProperties}
   {rendererClass=&NiagaraRibbonRendererProperties, rendererIndex=1}
  ]
 }
 Steam001 {
  bEnabled=false
  emitterUpdateScript [
   &EmitterState()
   &FN_SpawnAnimTrail_v2(Number of Anim Trails=4)
  ]
  particleSpawnScript [
   &FN_SetSpawnGroup_v2(Number of Spawn Groups=4)
   /Game/Effects/Niagara/Modules/AnimTrail/FN_InitializeAnimTrail/FN_InitializeAnimTrailBoneIndecies()
   &FN_InitializeAnimTrailBoneIndecies(AnimTrail Selection=1, First Bone or Socket Index=1, Second Bone or Socket Index=1) {module=FN_InitializeAnimTrailBoneIndecies001}
   &FN_InitializeAnimTrailBoneIndecies(AnimTrail Selection=2, First Bone or Socket Index=2, Second Bone or Socket Index=2) {module=FN_InitializeAnimTrailBoneIndecies002}
   &FN_InitializeAnimTrailBoneIndecies(AnimTrail Selection=3, First Bone or Socket Index=3, Second Bone or Socket Index=3) {module=FN_InitializeAnimTrailBoneIndecies003}
   /Game/Effects/Niagara/Modules/AnimTrail/FN_InitializeAnimTrail/FN_InitializeAnimTrail_v2(AnimTrailWidth=3, Color:LinearColor=<rgba 1 1 1 0.15>, Lifetime=1.5)
   &AddVelocity(Velocity=<0 0 27>)
   &Material(PostMult=0.2, PreMult=0)
  ]
  particleUpdateScript [
   /SetVariables(Particles.EngineOwnerSpeed:NiagaraFloat=&VectorLength(Vector=$Engine.Owner.Velocity)) {isSetParameters=true}
   &ParticleState()
   &InheritVelocity(
    Module.Limit Speed=true
    Velocity Limit=500
    Velocity Scale=&MakeVector(
     X=&FloatFromCurve(FloatCurve=DataInterface() {propertyValues {Curve {keys [<at 0 0.75>,<at 0.01 0>]}}}, Scale Curve=0.5)
     Y=&FloatFromCurve(FloatCurve=DataInterface() {propertyValues {Curve {keys [<at 0 0.75>,<at 0.01 0>]}}}, Scale Curve=0.5)
     Z=&FloatFromCurve(FloatCurve=DataInterface() {propertyValues {Curve {keys [<at 0 0.75>,<at 0.01 0>]}}}, Scale Curve=0.1)
    )
   )
   /SetVariables(Particles.Speed:NiagaraFloat=&NormalizeFloat(Value=&VectorLength(Vector=$Particles.Velocity), Min=0, Max=200)) {isSetParameters=true}
   &SkeletalMeshLocation(
    Bone / Socket Index=$Particles.SpawnGroup
    Inherit Velocity=false
    MaskSpawnGroup=false
    Mesh Sampling Type=::NewEnumerator5
    Skeletal Mesh=$User.SkeletalMesh
    Socket Sampling Mode=::NewEnumerator5
   ) {enabled=false}
   &CurlNoiseForce(
    Noise Frequency=0.25
    Noise Strength=&FloatFromCurve(FloatCurve=DataInterface() {propertyValues {Curve {keys [<at 0 1>,<at 1 0>]}}}, Scale Curve=150)
    Pan Noise=true
    Pan Noise Field=<0 0 -0.5>
   )
   &AddVelocity(Velocity=<0 0 -0.2>) {module=AddVelocity001}
   &SolveForcesAndVelocity(Clamp Velocity=false, Limit Acceleration=false)
   &FN_UpdateAnimTrail_v2()
   &ScaleColor(
    Scale Alpha=&Multiply_Float(
     A=&FloatFromCurve(FloatCurve=DataInterface() {propertyValues {Curve {keys [{arriveTangent=0.002, interpMode=RCIM_Cubic, time=0.0401, value=0.4707}, <at 1.0011 0.002>]}}})
     B=&FloatFromCurve(
      FloatCurve=DataInterface() {propertyValues {Curve {keys [
          {arriveTangent=-2.1907, interpMode=RCIM_Cubic, leaveTangent=-0.886, tangentMode=RCTM_User, time=0, value=1}
          {arriveTangent=-1.1071, interpMode=RCIM_Cubic, leaveTangent=-1.1071, tangentMode=RCTM_User, time=1, value=0}
         ]}}}
      CurveIndex=$Particles.Speed
     )
    )
    ScaleA=true
    ScaleRGB=true
   )
   &ScaleColor(
    Scale Alpha=&Lerp_Float(
     A=1
     B=0
     Alpha=&NormalizeDistanceRange(
      Start Position:NiagaraPosition=$Engine.Owner.Position
      End Position:NiagaraPosition=$Particles.Position
      Distance:NiagaraFloat=50
      Invert Normalized Range:NiagaraBool=false
     )
    )
    ScaleA=true
    ScaleRGB=true
   ) {module=ScaleColor001}
   &ScaleRibbonWidth(Ribbon Width Scale=&Multiply_Float(
     A=&FloatFromCurve(FloatCurve=DataInterface() {propertyValues {Curve {keys [
          {arriveTangent=2.7871, interpMode=RCIM_Cubic, leaveTangent=2.7871, tangentMode=RCTM_User, time=-0.0033, value=0.0049}
          {arriveTangent=-0.0276, interpMode=RCIM_Cubic, leaveTangent=-0.0276, tangentMode=RCTM_User, time=0.1555, value=0.7121}
          {interpMode=RCIM_Cubic, tangentMode=RCTM_User, time=1.0011, value=0.002}
         ]}}})
     B=&FloatFromCurve(
      FloatCurve=DataInterface() {propertyValues {Curve {keys [<at 0 1 RCIM_Cubic>, {arriveTangent=-1.2414, interpMode=RCIM_Cubic, leaveTangent=-1.2414, tangentMode=RCTM_User, time=1, value=0}]}}}
      CurveIndex=$Particles.Speed
      Scale Curve=2
     )
    )) {enabled=false}
   &KillParticles(Kill Particles=&SetBoolByFloatComparison(A=500, B=$Particles.EngineOwnerSpeed, Comparison Type=::NewEnumerator6))
  ]
  rendererClasses [&NiagaraRibbonRendererProperties]
  renderers [{rendererClass=&NiagaraRibbonRendererProperties}]
 }
}
summary {
 systemName=NS_Backpack_PumpkinSpice
 emitters [
  {bEnabled=false, emitterName=Steam001, rendererClasses [&NiagaraRibbonRendererProperties]}
  {emitterName=FN_AttachedRibbon_v3, rendererClasses [&NiagaraRibbonRendererProperties]}
 ]
 userVariables [
  User.GlobalOffset {type=Vector3f, description="", defaultValue {struct=Vector3f, value=<0 0 0>}}
  User.RibbonIDOffset {type=NiagaraInt32}
  User.SkeletalMesh {
   type=NiagaraDataInterfaceSkeletalMesh
   defaultValue {
    struct=NiagaraExt_VariableValue_DataInterface
    value {dataInterface=&self:NiagaraDataInterfaceSkeletalMesh_0}
   }
  }
 ]
}
