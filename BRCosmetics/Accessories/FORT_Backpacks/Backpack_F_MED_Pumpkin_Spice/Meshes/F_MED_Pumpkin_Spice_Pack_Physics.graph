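# Physics asset for the Pumpkin Spice backpack mesh, letting four cup bones jiggle on the pack.
# Each dyn_cup bone has a capsule body, and the pack bone has a small capsule that anchors them.
# Constraints tie each cup directly to the pack with tight swing and twist limits, so the cups wobble only a few degrees independently.
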PhysicsAsset /BRCosmetics/Accessories/FORT_Backpacks/Backpack_F_MED_Pumpkin_Spice/Meshes/F_MED_Pumpkin_Spice_Pack_Physics
bodies {
 dyn_cup1 [{shapeType=Capsule, length=6, radius=4, center=<5.107 0.6725 0.409>, rotation=<rot 87.8906 175.1423 180>}]
 dyn_cup2 [{shapeType=Capsule, length=6, radius=4.3, center=<4.9116 0.6747 -0.4197>, rotation=<rot 90 0 -5.9068>}]
 dyn_cup3 [{shapeType=Capsule, length=6, radius=4, center=<5.1119 0.9096 0.0411>, rotation=<rot 86.8347 6.3667 0>}]
 dyn_cup4 [{shapeType=Capsule, length=6, radius=4, center=<4.4263 0.2737 0.4757>, rotation=<rot 87.4697 -4.6538 0>}]
 pack [{shapeType=Capsule, length=1, radius=1}]
}
constraints [
 {bone1Name=dyn_cup1, bone2Name=pack, swing1LimitDegrees=5.5, swing2LimitDegrees=5.5, twistLimitDegrees=3}
 {bone1Name=dyn_cup2, bone2Name=pack, swing1LimitDegrees=5.5, swing2LimitDegrees=5.5, twistLimitDegrees=3}
 {bone1Name=dyn_cup3, bone2Name=pack, swing1LimitDegrees=5, swing2LimitDegrees=5, twistLimitDegrees=3}
 {bone1Name=dyn_cup4, bone2Name=pack, swing1LimitDegrees=5.5, swing2LimitDegrees=5.5, twistLimitDegrees=3}
]
